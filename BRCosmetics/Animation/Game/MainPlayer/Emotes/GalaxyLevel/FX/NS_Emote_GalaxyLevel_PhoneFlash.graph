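# Niagara system for the phone camera flash in the GalaxyLevel emote, a brief flare burst fired near start.
# FlareBurst scales size and alpha over life, derives star, streak, anamorphic, halo and glow attributes from that base, and feeds sprite, light and component renderers.
# The light sits midway between the sampled User.SkeletalMesh position and the simulation position, with intensity scaled by time of day.
# A camera-position-driven offset and the debug_Pos emitter are disabled, leaving a fixed camera offset.
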NiagaraSystem /BRCosmetics/Animation/Game/MainPlayer/Emotes/GalaxyLevel/FX/NS_Emote_GalaxyLevel_PhoneFlash
dependencies {
 usedDataInterfaces [&NiagaraDataInterfaceCamera, &NiagaraDataInterfaceCurve]
 usedDynamicInputs []
 usedModules [&SystemState]
}
emitters {
 FlareBurst {
  emitterUpdateScript [
   &EmitterState()
   &SpawnBurst_Instantaneous(Spawn Time=0.01, Use Spawn Probability=false, UseLoopCountLimit=false)
   /SetVariables(Emitter.Cam Pos:NiagaraPosition=&ReturnCameraVector(Camera Property=::NewEnumerator0, Coordinate Space=::Simulation)) {isSetParameters=true, module=SetVariables_#1}
  ]
  particleSpawnScript [
   &InitializeParticle(
    Lifetime Max=0.23
    Lifetime Min=0.18
    Lifetime Mode=::NewEnumerator1
    Mass Max=2
    Mass Mode=::NewEnumerator3
    Position Offset=<3 0 0>
    Position Offset Coordinate Space=::Local
    Sprite Rotation Mode=::NewEnumerator2
    Sprite Size Mode=::NewEnumerator3
    Uniform Sprite Size=&RandomRangeFloat(Minimum=10, Maximum=25)
    UsePositionOffset=true
    Write Material Random=true
   )
   &CameraOffset(Camera Offset Amount=&Lerp_Float(
     A=30
     B=-30
     Alpha=&Lerp_Float(
      A=0
      B=1
      Alpha=&self:ScratchDynamicInput(Cam Position:NiagaraPosition=$Emitter.Cam Pos)
     )
    )) {enabled=false}
  ]
  particleUpdateScript [
   &ParticleState()
   &CameraOffset(Camera Offset Amount=50) {module=CameraOffset001}
   &SampleSkeletalMesh(
    AdditionalYawPitchRoll=false
    Bone Sampling Mode:ENiagaraBoneSamplingMode=::NewEnumerator4
    Skeletal Mesh:NiagaraDataInterfaceSkeletalMesh=$User.SkeletalMesh
   ) {module=SampleSkeletalMesh001}
   &ScaleColor(
    Scale Alpha=&FloatFromCurve(
     FloatCurve=DataInterface() {propertyValues {Curve {keys [
         <at 0 0 RCIM_Cubic>
         {arriveTangent=0.1895, interpMode=RCIM_Cubic, leaveTangent=0.1895, tangentMode=RCTM_User, time=0.1, value=1}
         {arriveTangent=-0.0202, interpMode=RCIM_Cubic, leaveTangent=-0.0202, tangentMode=RCTM_User, time=0.7, value=1}
         <at 1.0011 0.002 RCIM_Cubic>
        ]}}}
     Scale Curve=10
    )
    ScaleA=true
    ScaleRGB=true
   )
   &ScaleSpriteSize(Scale Factor=&Vector2DFromFloat(Value=&FloatFromCurve(FloatCurve=DataInterface() {propertyValues {Curve {keys [
          {interpMode=RCIM_Cubic, tangentMode=RCTM_User, time=0, value=0}
          {arriveTangent=-0.0009, interpMode=RCIM_Cubic, leaveTangent=-0.0009, tangentMode=RCTM_User, time=0.2, value=1}
          {tangentMode=RCTM_User, time=1, value=0}
         ]}}})))
   /SetVariables(
    Particles.StarColor:LinearColor=&MultiplyLinearColor(Linear Color=$Particles.Color, Scale Factor=&MakeVector4_FromLinearColor(LinearColor=<rgba 1 1 1 1>))
    Particles.StarRotation:NiagaraFloat=&RandomRangeFloat(Minimum=0, Maximum=360)
    Particles.StarSize:Vector2f=&Multiply_Vector2(A=$Particles.SpriteSize, B=<2 2>)
   ) {isSetParameters=true}
   /SetVariables(
    Particles.StreaksColor:LinearColor=&MultiplyLinearColor(
     Linear Color=$Particles.Color
     Scale Factor=&MakeVector4_FromLinearColor(LinearColor=<rgba 0.6 0.8 1 0.01>)
    )
    Particles.StreaksRotation:NiagaraFloat=&RandomRangeFloat(Minimum=0, Maximum=360)
    Particles.StreaksSize:Vector2f=&Multiply_Vector2(A=$Particles.SpriteSize, B=<7 7>)
   ) {isSetParameters=true}
   /SetVariables(
    Particles.AnamorphicColor:LinearColor=&MultiplyLinearColor(
     Linear Color=$Particles.Color
     Scale Factor=&MakeVector4_FromLinearColor(LinearColor=<rgba 0.8 0.9 1 0.01>)
    )
    Particles.AnamorphicRotation=90
    Particles.AnamorphicSize:Vector2f=&Multiply_Vector2(A=$Particles.SpriteSize, B=<3 9>)
   ) {isSetParameters=true}
   /SetVariables(
    Particles.HaloColor:LinearColor=&MultiplyLinearColor(Linear Color=$Particles.Color, Scale Factor=&MakeVector4_FromLinearColor(LinearColor=<rgba 1 1 1 0.025>))
    Particles.HaloSize:Vector2f=&Multiply_Vector2(A=$Particles.SpriteSize, B=<5 5>)
   ) {isSetParameters=true}
   /SetVariables(
    Particles.HaloColor_02:LinearColor=&MultiplyLinearColor(Linear Color=$Particles.Color, Scale Factor=&MakeVector4_FromLinearColor(LinearColor=<rgba 1 1 1 0.05>))
    Particles.HaloSize_02:Vector2f=&Multiply_Vector2(A=$Particles.SpriteSize, B=<1.5 1.5>)
   ) {isSetParameters=true}
   /SetVariables(
    Particles.AirGlowColor:LinearColor=&MultiplyLinearColor(Linear Color=$Particles.Color, Scale Factor=&MakeVector4_FromLinearColor(LinearColor=<rgba 1 1 1 0.005>))
    Particles.AirGlowSize:Vector2f=&Multiply_Vector2(A=$Particles.SpriteSize, B=<4 4>)
   ) {isSetParameters=true}
   &FN_Light_Attributes(
    Color=<rgba 1 1 1 0.2>
    Create Color Attribute=true
    Create Exponent Attribute=true
    Create Position Attribute=true
    Create Radius Attribute=true
    Enable Particle Light Enable=true
    Light Enabled=true
    Position=/Niagara/DynamicInputs/Lerp/Lerp_Position(
     A:NiagaraPosition=$Particles.SampleSkeletalMesh001.SampledPosition
     B:NiagaraPosition=/Niagara/DynamicInputs/Helpers/SimulationPosition()
     Alpha:NiagaraFloat=0.5
    )
    Radius=300
    Scale Color=&FloatFromCurve(
     FloatCurve=DataInterface() {propertyValues {Curve {keys [
         {arriveTangent=9.0725, interpMode=RCIM_Cubic, leaveTangent=9.0725, tangentMode=RCTM_User, time=0, value=0}
         {arriveTangent=-2.5249, interpMode=RCIM_Cubic, leaveTangent=-2.5249, tangentMode=RCTM_User, time=0.1408, value=0.9847}
         {arriveTangent=-0.0405, interpMode=RCIM_Cubic, leaveTangent=-0.0405, tangentMode=RCTM_User, time=1, value=0}
        ]}}}
     Scale Curve=10
    )
   )
   /SetVariables(Particles.TempPosition:Vector3f=$Particles.FN_Light_Attributes.Position) {isSetParameters=true}
   &FN_Gen5_Light(
    Attenutation=600
    Exponent=8
    Intensity:NiagaraFloat=&FNTimeOfDayScalar(Night Scale=2, Morning Scale=6, Day Scale=8, Evening Scale=5)
   )
  ]
  rendererClasses [
   &NiagaraSpriteRendererProperties
   &NiagaraLightRendererProperties
   &NiagaraComponentRendererProperties
  ]
  renderers [
   {rendererClass=&NiagaraSpriteRendererProperties}
   {rendererClass=&NiagaraSpriteRendererProperties, rendererIndex=1}
   {rendererClass=&NiagaraSpriteRendererProperties, rendererIndex=2}
   {rendererClass=&NiagaraSpriteRendererProperties, rendererIndex=3}
   {rendererClass=&NiagaraLightRendererProperties, rendererIndex=4}
   {rendererClass=&NiagaraComponentRendererProperties, rendererIndex=5}
  ]
 }
 debug_Pos {
  bEnabled=false
  emitterUpdateScript [
   &EmitterState()
   &SpawnBurst_Instantaneous(Use Spawn Probability=false, UseLoopCountLimit=false)
   /SetVariables(Emitter.Cam Pos:NiagaraPosition=&ReturnCameraVector(Camera Property=::NewEnumerator0, Coordinate Space=::Simulation)) {isSetParameters=true, module=SetVariables_#1}
  ]
  particleSpawnScript [
   &InitializeParticle(
    Color Mode=::NewEnumerator0
    Mesh Scale Mode=::NewEnumerator3
    Sprite Size Mode=::NewEnumerator3
    Uniform Sprite Size=100
    UsePositionOffset=false
    Write Material Random=true
   )
   &CameraOffset(Camera Offset Amount=25)
   &DynamicMaterialParameters(
    Index 0 Param 1=&Lerp_Float(
     A=0
     B=500
     Alpha=&Lerp_Float(
      A=0
      B=1
      Alpha=&self:ScratchDynamicInput(Cam Position:NiagaraPosition=$Emitter.Cam Pos)
     )
    )
    Module.Param0WriteEnabled=true
    Module.Param1WriteEnabled=true
    Module.Param2WriteEnabled=true
    Module.Param3WriteEnabled=true
   )
  ]
  particleUpdateScript [&ParticleState()]
  rendererClasses [&NiagaraSpriteRendererProperties]
  renderers [{rendererClass=&NiagaraSpriteRendererProperties}]
 }
}
summary {
 systemName=NS_Emote_GalaxyLevel_PhoneFlash
 emitters [
  {
   emitterName=FlareBurst
   rendererClasses [
    &NiagaraSpriteRendererProperties
    &NiagaraLightRendererProperties
    &NiagaraComponentRendererProperties
   ]
  }
  {bEnabled=false, emitterName=debug_Pos, rendererClasses [&NiagaraSpriteRendererProperties]}
 ]
 userVariables [
  User.bIsFrontEndPreview {type=NiagaraBool}
  User.Color_01 {type=LinearColor, description="", defaultValue {struct=LinearColor, value=<rgba 1 1 1 1>}}
  User.SkeletalMesh {
   type=NiagaraDataInterfaceSkeletalMesh
   defaultValue {
    struct=NiagaraExt_VariableValue_DataInterface
    value {dataInterface=&self:NiagaraDataInterfaceSkeletalMesh_0}
   }
  }
 ]
}
